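# Parent unlit, masked material for the MembraneInfinity vehicle wheel cosmetic, producing a colorable emissive infinity effect.
# An offset function samples a tiled, shiftable spoke texture; its red channel scales a desaturated, hue-shifted CustomColor, and green blends toward squared CustomColor times EmissiveColor.
# A radial falloff from UV center drives opacity mask.
# The attributes then pass through shared vehicle damage and de-rez transition functions.
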Material /VehicleCosmetics/Wheels/MembraneInfinity/Materials/M_Parent_MembraneInfinity_01
graph {
 properties {
  bAutomaticallySetUsageInEditor=false
  bUseMaterialAttributes=true
  bUsedWithMeshDeformer=true
  bUsedWithSkeletalMesh=true
  blendMode=BLEND_Masked
  previewMesh=/VehicleCosmetics/Wheels/MembraneInfinity/Mesh/SK_Wheel_MembraneInfinity
  shadingModel=MSM_Unlit
  usedShadingModels=Unlit
 }
 nodes [
  n1 = TextureObjectParameter() {parameterName=Param, samplerType=SAMPLERTYPE_LinearColor, texture=/VehicleCosmetics/Wheels/MembraneInfinity/Textures/MembraneSpoke_INF_M_T}
  n2 = Multiply(A=TextureCoordinate(), B=ScalarParameter() {defaultValue=1, parameterName=Tile})
  n3 = ScalarParameter() {parameterName=Shift}
  n4 = Clamp(None=MaterialFunctionCall!(
    Texture=@n1
    Offset=ScalarParameter() {defaultValue=-1.5, parameterName=Offset}
    UVs=Add(A=@n2, B=@n3)
    RotateUVs=StaticBool()
    GradientFalloff
    GradientDarkness=Constant() {r=0.1}
   ) {materialFunction=/VehicleCosmetics/Wheels/ClockGear_SE/Materials/MF_Infinity_OffsetFunction}.Result)
  n5 = VectorParameter() {parameterName=CustomColor, defaultValue=<rgba 1 0.1 0.2 1>}
  n6 = Multiply(A=@n5.RGB, B=@n5.RGB)
  n7 = Desaturation(None=@n6, Fraction=ScalarParameter() {defaultValue=0.25, parameterName=DesaturationColor})
  n8 = ScalarParameter() {parameterName=Angle}
  n9 = LinearInterpolate(
   A=Multiply(
    A=ComponentMask(None=@n4) {r=true}
    B=MaterialFunctionCall!(Color=@n7, Angle=@n8) {materialFunction=/VehicleCosmetics/Wheels/Concept_Inverted/Materials/MF_HUE_Shift}.Result
   )
   B=Multiply(A=@n6, B=ScalarParameter() {defaultValue=1, parameterName=EmissiveColor})
   Alpha=Add(A=Constant() {r=0.01}, B=ComponentMask(None=@n4) {g=true})
  )
  n10 = Add(A=TextureCoordinate(), B=Constant() {r=-0.5})
  n11 = Clamp(None=OneMinus(None=DotProduct(A=@n10, B=@n10)))
  MaterialFunctionCall(In=MaterialFunctionCall(Material=SetMaterialAttributes!(MaterialAttributes, Emissive Color=@n9, Opacity Mask=@n11) {attributeSetTypes [#1, #2]}) {materialFunction=&MF_BaseVehicle_Damage}.Result) {materialFunction=&MF_VehicleDeRez_Transition}
  TextureSampleParameter2D() {parameterName=MembraneSpoke, texture=&MembraneSpoke_INF_M_T}
 ]
}
