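# Parent material for the MembraneInfinity vehicle cosmetic wheel, built for skeletal meshes and mesh deformers using material attributes.
# MF_WheelParent_Simple supplies the base surface, whose normal is blended toward a second "Special Effects" normal map parameter by vertex color blue.
# The modified attributes then pass through MF_BaseVehicle_Damage and MF_VehicleDeRez_Transition to add shared vehicle damage and de-rez transition handling.
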Material /VehicleCosmetics/Wheels/MembraneInfinity/Materials/M_WheelParent_MembraneInfinity
graph {
 properties {
  bAutomaticallySetUsageInEditor=false
  bUseMaterialAttributes=true
  bUsedWithMeshDeformer=true
  bUsedWithSkeletalMesh=true
  pixelDepthOffsetMode=PDOM_Legacy
  previewMesh=/VehicleCosmetics/Wheels/MembraneInfinity/Mesh/SK_Wheel_MembraneInfinity
  usedShadingModels=Default Lit
 }
 nodes [
  n1 = GetMaterialAttributes(None=MaterialFunctionCall() {materialFunction=&MF_WheelParent_Simple}.MaterialAttributes) {attributeGetTypes [#1, #2]}
  n2 = TextureSampleParameter2D() {
   group=5_Special Effects
   parameterName=Normal_2
   samplerType=SAMPLERTYPE_Normal
   texture=/VehicleCosmetics/Wheels/MembraneInfinity/Textures/T_MembraneInfinity_SE2_N
  }
  n3 = LinearInterpolate(A=@n1.Normal, B=@n2.RGB, Alpha=LinearInterpolate(A=Constant(), B=Constant() {r=1}, Alpha=VertexColor().B))
  MaterialFunctionCall(In=MaterialFunctionCall(Material=SetMaterialAttributes!(MaterialAttributes=@n1.MaterialAttributes, Emissive Color=@n1.MaterialAttributes, Normal=@n3) {attributeSetTypes [#1, #2]}) {materialFunction=&MF_BaseVehicle_Damage}.Result) {materialFunction=&MF_VehicleDeRez_Transition}
 ]
}
