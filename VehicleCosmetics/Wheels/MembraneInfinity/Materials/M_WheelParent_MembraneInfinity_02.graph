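# Masked parent material for MembraneInfinity skeletal-mesh vehicle wheels, adding animated glow effects to the shared simple wheel base.
# An infinity-offset function borrowed from ClockGear_SE builds a depth-like glow, combined with tintable, sine-pulsed rim emissive through texture and vertex-color masks.
# Additional normal, color, metallic and roughness values blend over the base, while a Lods switch controls masked cutouts.
# The result passes through vehicle damage and de-rez transition functions.
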Material /VehicleCosmetics/Wheels/MembraneInfinity/Materials/M_WheelParent_MembraneInfinity_02
graph {
 properties {
  bAutomaticallySetUsageInEditor=false
  bUseMaterialAttributes=true
  bUsedWithMeshDeformer=true
  bUsedWithSkeletalMesh=true
  blendMode=BLEND_Masked
  pixelDepthOffsetMode=PDOM_Legacy
  previewMesh=/VehicleCosmetics/Wheels/MembraneInfinity/Mesh/SK_Wheel_MembraneInfinity
  usedShadingModels=Default Lit
 }
 nodes [
  n1 = TextureObjectParameter() {
   group=5_Special Effects
   parameterName=Bump_Ofrset_MAsk
   samplerType=SAMPLERTYPE_LinearColor
   texture=/VehicleCosmetics/Wheels/MembraneInfinity/Textures/T_MembraneInfinity_Emissive1
  }
  n2 = ScalarParameter() {defaultValue=-0.01, group=5_Special Effects, parameterName=Bump_Offset_Amount, sortPriority=1}
  n3 = ScalarParameter() {defaultValue=0.0003, group=5_Special Effects, parameterName=Gradient_Darkness, sortPriority=1}
  n4 = TextureSample() {samplerType=SAMPLERTYPE_LinearColor, texture=/VehicleCosmetics/Wheels/MembraneInfinity/Textures/T_MembraneInfinity_Effect_Mask}
  n5 = VertexColor()
  n6 = LinearInterpolate(
   B=ComponentMask(None=MaterialFunctionCall!(Texture=@n1, Offset=@n2, UVs, RotateUVs, GradientFalloff, GradientDarkness=@n3) {materialFunction=/VehicleCosmetics/Wheels/ClockGear_SE/Materials/MF_Infinity_OffsetFunction}.Result) {r=true}
   Alpha=Subtract(A=Subtract(A=@n4.R, B=@n5.G), B=@n5.B)
  )
  n7 = Multiply(A=Multiply(A=@n6, B=@n6), B=ScalarParameter() {defaultValue=3, group=5_Special Effects, parameterName=PainBoostWithoutEmissive})
  Add(A=@n7, B=@n7)
  ComponentMask(None=NamedRerouteUsage()) {g=true}
  LinearInterpolate(B=@n4.R, Alpha=@n5.R)
  n8 = VertexColor()
  LinearInterpolate(A=Constant(), B=@n4.G, Alpha=@n8.R)
  n9 = MaterialFunctionCall() {materialFunction=&MF_WheelParent_Simple}
  n10 = GetMaterialAttributes(None=@n9.MaterialAttributes) {attributeGetTypes [
    #1
    #2
    #3
    #4
    #5
    #6
   ]}
  n11 = TextureSampleParameter2D() {
   group=5_Special Effects
   parameterName=Normal_Additional
   samplerType=SAMPLERTYPE_Normal
   texture=/VehicleCosmetics/Wheels/MembraneInfinity/Textures/T_MembraneInfinity_SE2_N
  }
  n12 = NamedRerouteUsage()
  n13 = VectorParameter() {group=5_Special Effects, parameterName=Custom_Color_Aditional, defaultValue=<rgba 0.0698 0.0698 0.0698 1>}
  n14 = OneMinus(None=Add(A=Add(A=@n8.R, B=@n8.R), B=@n8.B))
  n15 = LinearInterpolate(A=@n10.Opacity Mask, B=NamedRerouteUsage(), Alpha=@n14)
  n16 = NamedRerouteUsage()
  n17 = SetMaterialAttributes!(
   MaterialAttributes=@n10.MaterialAttributes
   Emissive Color=LinearInterpolate(A=@n10.Emissive Color, B=NamedRerouteUsage(), Alpha=Constant() {r=1})
   Normal=LinearInterpolate(A=@n10.Normal, B=@n11.RGB, Alpha=@n12)
   Base Color=LinearInterpolate(A=@n10.Base Color, B=@n13.RGB, Alpha=@n12)
   Metallic=LinearInterpolate(A=@n10.Metallic, B=ScalarParameter() {defaultValue=1, group=5_Special Effects, parameterName=Metallic_Additional}, Alpha=@n12)
   Roughness=LinearInterpolate(A=@n10.Roughness, B=ScalarParameter() {defaultValue=0.25, group=5_Special Effects, parameterName=Roughness_Additional}, Alpha=@n12)
   Opacity Mask=StaticSwitchParameter(True=LinearInterpolate(A=@n15, B=@n16, Alpha=@n8.B), False=Constant() {r=1}) {group=5_Special Effects, parameterName=Lods}
  ) {attributeSetTypes [
    #1
    #2
    #3
    #4
    #5
    #6
   ]}
  MaterialFunctionCall(In=MaterialFunctionCall(Material=@n17) {materialFunction=&MF_BaseVehicle_Damage}.Result) {materialFunction=&MF_VehicleDeRez_Transition}
  Multiply(A=@n4.B)
  Multiply(A=@n8.R)
  n18 = VectorParameter() {group=5_Special Effects, parameterName=PaintColor, defaultValue=<rgba 0.72 0.33 1 1>}
  n19 = StaticSwitchParameter(True=@n18.RGB, False=VectorParameter() {group=5_Special Effects, parameterName=RingEmissiveColor, defaultValue=<rgba 0 0 0 1>}.RGB) {group=5_Special Effects, parameterName=PaintTone_or_DuoTone}
  n20 = TextureSampleParameter2D() {
   group=5_Special Effects
   parameterName=Mask_Emissive
   samplerType=SAMPLERTYPE_LinearColor
   texture=/VehicleCosmetics/Wheels/MembraneInfinity/Textures/T_MembraneInfinity_SE2_RGB
  }
  n21 = Constant()
  n22 = Add(A=Multiply(A=@n19, B=@n20.A), B=LinearInterpolate(A=@n21, B=@n18.RGB, Alpha=@n7))
  n23 = Sine(None=Multiply(A=Time(), B=Constant() {r=0.333}))
  n24 = Clamp(None=Multiply(A=Multiply(A=@n20.R, B=@n23), B=@n19))
  n25 = Subtract(A=StaticSwitchParameter(True=@n4.B, False=@n20.G) {group=5_Special Effects, parameterName=Lods}, B=NamedRerouteUsage())
  n26 = ScalarParameter() {defaultValue=0.1, group=5_Special Effects, parameterName=CoreGlowBack_PowerExponent, sortPriority=2}
  n27 = Multiply(A=@n25, B=@n26)
  n28 = Add(A=@n7, B=@n27)
  n29 = LinearInterpolate(A=@n21, B=@n18.RGB, Alpha=Add(A=@n28, B=@n28))
  n30 = Multiply(
   A=Multiply(A=@n27, B=@n19)
   B=ScalarParameter() {defaultValue=1, group=5_Special Effects, parameterName=CoreGlowBack_Multiplier, sortPriority=2}
  )
  n31 = LinearInterpolate(
   A=Multiply(A=Add(A=@n22, B=@n24), B=ScalarParameter() {defaultValue=5, group=5_Special Effects, parameterName=Rim_Emissive})
   B=Multiply(A=Add(A=@n29, B=@n30), B=ScalarParameter() {defaultValue=1, group=5_Special Effects, parameterName=Height_Emissive_Brightness})
   Alpha=OneMinus(None=@n8.R)
  )
  n32 = VectorParameter() {group=5_Special Effects, parameterName=Emissive_Effect_Mask_Color, defaultValue=<rgba 0 0.1356 0.5711 1>}
  n33 = Multiply(A=NamedRerouteUsage(), B=@n32.RGB)
  n34 = NamedRerouteUsage()
  n35 = LinearInterpolate(B=Multiply(A=@n33), Alpha=Add(A=@n20.R, B=@n4.G))
  NamedRerouteDeclaration(None=Max(A=LinearInterpolate(A=@n31, B=@n33, Alpha=@n34), B=LinearInterpolate(B=@n35, Alpha=@n34))) {name=Emissive, nodeColor=<rgba 1 0.75 0 1>}
  NamedRerouteDeclaration(None=@n5.B) {name=BC_RC_Blue, nodeColor=<rgba 0.1016 1 0 1>}
  NamedRerouteDeclaration(None=@n20.B) {name=EmissiveMask_Blue, nodeColor=<rgba 1 0 0.9844 1>}
  NamedRerouteDeclaration(None=@n5.G) {name=VC_Green, nodeColor=<rgba 0 1 0.9297 1>}
  NamedRerouteDeclaration(None=@n4.A) {name=Alpha_Lods, nodeColor=<rgba 0.125 0 1 1>}
  NamedRerouteDeclaration(None=@n20.A) {name=Alpha_Lods_2, nodeColor=<rgba 1 0 0.0234 1>}
  NamedRerouteDeclaration(None=@n14) {name=VC_Inverted, nodeColor=<rgba 1 0.2812 0 1>}
  NamedRerouteUsage()
  NamedRerouteUsage()
  NamedRerouteUsage()
  NamedRerouteUsage()
  NamedRerouteUsage()
  OneMinus(None=@n9.VertexColorR)
  Power(Base=@n25, Exp=@n26)
  StaticSwitchParameter(True=@n20.B, False=@n20.B) {group=5_Special Effects, parameterName=Lods}
 ]
}
